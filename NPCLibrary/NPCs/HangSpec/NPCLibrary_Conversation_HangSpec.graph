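# Conversation database for the HangSpec NPC (Yoo-mi), serving as her top-level dialogue when a player talks to her.
# The entry point leads to a greeting speech node that picks contextual lines by priority: first-ever meeting, match conversation count, or specific player character cosmetics, with a default fallback.
# The greeting then branches through link nodes to shared NPC library quests, basic services, and HangSpec-specific dialogue entries.
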ConversationDatabase /NPCLibrary/NPCs/HangSpec/NPCLibrary_Conversation_HangSpec
entry_points [{tag=Conversation.Entry.HangSpec, destinations [#1]}]
nodes [
 ConversationEntryPointNode_7 {
  id=#1
  class=ConversationEntryPointNode
  outputs [#2]
  properties {entryTag {tagName=Conversation.Entry.HangSpec}}
 }
 Speech_C_4 {
  id=#2
  class=Speech_C
  outputs [#3, #4, #5]
  properties {
   nodeName=""
   generalConfig {
    defaultMessage=It's an honor to serve O.X.R.
    contextualMessages [
     {
      message="Hi! I'm Yoo-mi. It's nice to meet you!"
      contextRequirements [{requirement {class=FortControllerRequirement_IsFirstEverConversationWithNPC}}]
      priority {curve {rowName=Tandem.Speech.Priority.FirstEverInteraction}}
     }
     {
      message=It's an honor to serve O.X.R.
      contextRequirements [{requirement {class=FortControllerRequirement_CheckNPCMatchConversationCount, properties {nPCUniqueIdTag {tagName=AISpawnerData.Type.Tandem.HangSpec}}}}]
      priority {curve {rowName=None}}
     }
     {
      message=Up for some karaoke after this?
      contextRequirements [{requirement {class=FortControllerRequirement_CheckNPCMatchConversationCount, properties {nPCUniqueIdTag {tagName=AISpawnerData.Type.Tandem.HangSpec}}}}]
      priority {curve {rowName=None}}
     }
     {
      message="Your backstage passes for my next tour? Um, totally in the mail. Yeah."
      contextRequirements [{requirement {class=FortControllerRequirement_HasCID, properties {allowedCIds [/BRCosmetics/Athena/Items/Cosmetics/Characters/Character_TalonPane_Flail]}}}]
      priority {value=2, curve {curveTable=null, rowName=None}}
     }
     {
      message="I won't let you down, General."
      contextRequirements [{requirement {class=FortControllerRequirement_HasCID, properties {allowedCIds [/BRCosmetics/Athena/Items/Cosmetics/Characters/Character_SwoopClaspPry]}}}]
      priority {value=2, curve {curveTable=null, rowName=None}}
     }
    ]
   }
  }
 }
 ConversationLinkNode_2 {id=#5, class=ConversationLinkNode, properties {remoteEntryTag {tagName=Conversation.Entry.HangSpec.Dialogue}}}
 ConversationLinkNode_0 {id=#4, class=ConversationLinkNode, properties {remoteEntryTag {tagName=Conversation.Entry.BasicServices}}}
 ConversationLinkNode_1 {id=#3, class=ConversationLinkNode, properties {remoteEntryTag {tagName=Conversation.Entry.NPCLibrary.Quests}}}
]
