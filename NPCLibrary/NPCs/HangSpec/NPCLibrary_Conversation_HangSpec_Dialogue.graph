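# Conversation database for the "Ask about..." dialogue of the HangSpec NPC, who calls herself Yoo-mi, an idol and O.X.R. member.
# A single entry tag leads to a greeting that branches into three topics (idol life, sharpshooting, HUNTR/X) plus a Back exit.
# Each topic is a linear chain of Speech nodes, and all three converge on one shared ReturnToStart node labeled BACK.
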ConversationDatabase /NPCLibrary/NPCs/HangSpec/NPCLibrary_Conversation_HangSpec_Dialogue
entry_points [{tag=Conversation.Entry.HangSpec.Dialogue, destinations [#1]}]
nodes [
 ConversationEntryPointNode_7 {
  id=#1
  class=ConversationEntryPointNode
  outputs [#2]
  properties {entryTag {tagName=Conversation.Entry.HangSpec.Dialogue}}
 }
 Speech_C_0 {
  id=#2
  class=Speech_C
  outputs [#3, #4, #5, #6]
  properties {nodeName="", softTaskIcon=&T_UI_NPC_question, generalConfig {defaultMessage=Did you want me to sign something?}}
  sub_nodes [ConversationChoiceNode_10 {class=ConversationChoiceNode, properties {defaultChoiceDisplayText=Ask about...}}]
 }
 Speech_C_6 {
  id=#6
  class=Speech_C
  outputs [#7]
  properties {
   nodeName=""
   softTaskIcon=&T_UI_NPC_question
   generalConfig {defaultMessage=Now those are some INCREDIBLE demon hunters.}
  }
  sub_nodes [ConversationChoiceNode_15 {class=ConversationChoiceNode, properties {defaultChoiceDisplayText=HUNTR/X}}]
 }
 Speech_C_8 {
  id=#7
  class=Speech_C
  outputs [#8]
  properties {nodeName="", generalConfig {defaultMessage="As soon as I heard we were dealing with a big bad Dark Presence, I was like, \"hmm...\""}}
 }
 Speech_C_9 {
  id=#8
  class=Speech_C
  outputs [#9]
  properties {nodeName="", generalConfig {defaultMessage="\"...Didn't I hear Rumi, Mira, and Zoey talking about a Dark Presence a while back?\""}}
 }
 Speech_C_10 {
  id=#9
  class=Speech_C
  outputs [#10]
  properties {nodeName="", generalConfig {defaultMessage="If they can help, they will. And if they can't help... we're in serious trouble."}}
 }
 ReturnToStart_C_23 {
  id=#10
  class=ReturnToStart_C
  properties {softTaskIcon=&T_UI_NPC_Back}
  sub_nodes [ConversationChoiceNode_12 {class=ConversationChoiceNode, properties {defaultChoiceDisplayText=BACK}}]
 }
 Speech_C_2 {
  id=#5
  class=Speech_C
  outputs [#11]
  properties {
   nodeName=""
   softTaskIcon=&T_UI_NPC_question
   generalConfig {defaultMessage="My dad taught me how to be a marksman. He served in O.X.R. while it was dormant. "}
  }
  sub_nodes [ConversationChoiceNode_14 {class=ConversationChoiceNode, properties {defaultChoiceDisplayText=Sharpshooting}}]
 }
 Speech_C_3 {
  id=#11
  class=Speech_C
  outputs [#12]
  properties {nodeName="", generalConfig {defaultMessage="We'd either be hanging out at the range, or he'd be taking me to my music lessons."}}
 }
 Speech_C_1 {
  id=#12
  class=Speech_C
  outputs [#13]
  properties {nodeName="", generalConfig {defaultMessage="He retired before the Bugs came back. And then my music career really took off. "}}
 }
 Speech_C_11 {
  id=#13
  class=Speech_C
  outputs [#10]
  properties {
   nodeName=""
   generalConfig {defaultMessage="But we kept up our tradition of father-daughter time at the range. We still go when we can, and he's basically my best friend."}
  }
 }
 Back_C_1 {
  id=#4
  class=Back_C
  sub_nodes [ConversationChoiceNode_13 {class=ConversationChoiceNode, properties {defaultChoiceDisplayText=BACK}}]
 }
 Speech_C_7 {
  id=#3
  class=Speech_C
  outputs [#14]
  properties {
   nodeName=""
   softTaskIcon=&T_UI_NPC_question
   generalConfig {defaultMessage=It's a dream. An honor. But also... confining.}
  }
  sub_nodes [ConversationChoiceNode_11 {class=ConversationChoiceNode, properties {defaultChoiceDisplayText=Being an Idol}}]
 }
 Speech_C_15 {
  id=#14
  class=Speech_C
  outputs [#15]
  properties {
   nodeName=""
   generalConfig {defaultMessage="I love being an artist. I LOVE my fans. But the world has strict perceptions of what an Idol is, and what they can do."}
  }
 }
 Speech_C_14 {
  id=#15
  class=Speech_C
  outputs [#16]
  properties {nodeName="", generalConfig {defaultMessage="I am a great singer and dancer, yes. But there's so much more to me than that."}}
 }
 Speech_C_5 {
  id=#16
  class=Speech_C
  outputs [#10]
  properties {nodeName="", generalConfig {defaultMessage="It's part of why I joined O.X.R. To prove that I'm more than an idol. I'm Yoo-mi!"}}
 }
]
